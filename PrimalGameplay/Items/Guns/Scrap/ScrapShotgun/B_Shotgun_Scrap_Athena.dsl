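; Weapon Blueprint for the Primal scrap shotgun (Athena), a child class that customizes the look of the gun's muzzle flash.
; It overrides the muzzle flash FX event, calls the parent, then sets a pale grey color and a raised alpha multiplier on the spawned muzzle particle system.
; The construction script only calls its parent.
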(event EventMuzzleFlashFX
  (params (Persistent%20Fire :type Boolean))
  (bind _spawned_muzzle_system (Variables|Effects|Muzzle|GetSpawned%20Muzzle%20System))
  (|CallParentFunction Persistent%20Fire)
  (Effects|Components|ParticleSystem|SetColorParameter _spawned_muzzle_system "Color" "(R=0.159969,G=0.194990,B=0.203125,A=1.000000)")
  (Effects|Components|ParticleSystem|SetFloatParameter _spawned_muzzle_system "AlphaMult" 1.5))

(fn ConstructionScript
  (|CallParentFunction))
